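; Behavior tree decorator for DADBRO AI that checks whether the actor in a configurable blackboard key is an airborne pawn.
; It casts that actor to FortPawn and maps its CharacterMovement mode to a boolean through a Select.
; Every mode except Walking and NavWalking counts as airborne, including Swimming, None and Custom.
; A MergeExecution node supplies the condition's return value.
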(fn PerformConditionCheckAI
  (params (OwnerController :type AIController%20Object%20Reference) (ControlledPawn :type Pawn%20Object%20Reference))
  (bind _asfort_pawn (Utilities|Casting|CastToFortPawn (AI|BehaviorTree|GetBlackboardValueasActor (Variables|Default|GetActorKeyToTest)))
    (:then
      (if (Utilities|Select :Index (Variables|/Script/Engine.CharacterMovementComponent|GetMovementMode (Variables|/Script/Engine.Character|GetCharacterMovement _asfort_pawn)) :MOVE_None true :MOVE_Walking false :MOVE_NavWalking false :MOVE_Falling true :MOVE_Swimming true :MOVE_Flying true :MOVE_Custom true)
        _))
    (:CastFailed
      _))
  (bind _fromtrueinput (Utilities|FlowControl|MergeExecution)
    (:Then
      (return _fromtrueinput))))
